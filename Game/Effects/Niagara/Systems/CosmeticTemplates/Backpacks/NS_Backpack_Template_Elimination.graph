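# Template Niagara system for backpack cosmetics that react to eliminations, showing several ways to drive sprite effects.
# Burst spawns particles when a value-change flag fires after the first frame and User.KillCount passes a comparison with zero.
# System Timeline time drives Rate's spawn-rate curve and IdleChange's persistent sprite size and color pulse.
# TimeStamp animates another persistent sprite's size and color from time elapsed since User.TimeStamp, avoiding respawns.
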NiagaraSystem /Game/Effects/Niagara/Systems/CosmeticTemplates/Backpacks/NS_Backpack_Template_Elimination
dependencies {
 usedDataInterfaces [&NiagaraDataInterfaceCurve, &NiagaraDataInterfaceColorCurve]
 usedDynamicInputs [&SetBoolByFloatComparison]
 usedModules [
  &self:SystemSpawnScript.SystemScriptSource.SystemScriptGraph.NiagaraNodeAssignment_0.SetVariables_#1
  &SystemState
  /Game/Effects/Niagara/Modules/Logic/UpdateValueStateChange
  &Timeline
 ]
}
emitters {
 Burst {
  emitterUpdateScript [
   &EmitterState()
   /Niagara/Modules/Emitter/SpawnByTrigger(
    Spawn Count=10
    Trigger:NiagaraBool=&Boolean_AndOperation(
     Bool1=&SetBoolByIntComparison(A=$User.KillCount, B=0, Comparison Type=::NewEnumerator0)
     Bool2=&Boolean_AndOperation(Bool1=$System.UpdateValueStateChange.Changed, Bool2=&Boolean_NotOperation(Bool=$Transient.FirstFrame))
    )
    TriggerIntervalSwitch=false
   )
   &SpawnPerFrame(
    Spawn:NiagaraBool=&Boolean_AndOperation(
     Bool1=&Boolean_AndOperation(
      Bool1=&Boolean_NotOperation(Bool=$System.UpdateValueStateChange.Changed)
      Bool2=&Boolean_NotOperation(Bool=$Transient.FirstFrame)
     )
     Bool2=$System.Reactive
    )
    Spawn Count=10
    Use Spawn Probability=false
   ) {enabled=false}
  ]
  particleSpawnScript [
   &InitializeParticle(Color Mode=::NewEnumerator0, Sprite Size Mode=::NewEnumerator3, Uniform Sprite Size=8, UsePositionOffset=false, Write Material Random=true)
   &AddVelocity(
    Module.Use Falloff Distance=false
    Override Seed=false
    Velocity Mode=::NewEnumerator1
    Velocity Speed=&RandomRangeFloat(Minimum=150, Maximum=300)
   )
  ]
  particleUpdateScript [
   &ParticleState()
   &Drag(Drag=2, Use Linear Drag=true, Use Rotational Drag=true)
   &GravityForce(Gravity=<0 0 -300>)
   &SolveForcesAndVelocity(Clamp Velocity=false, Limit Acceleration=false)
  ]
  rendererClasses [&NiagaraSpriteRendererProperties]
  renderers [{rendererClass=&NiagaraSpriteRendererProperties}]
 }
 IdleChange {
  emitterUpdateScript [
   &EmitterState(Life Cycle Mode=::NewEnumerator1, Loop Behavior=::NewEnumerator1, Loop Duration=1, UseLoopDelay=false)
   &SpawnBurst_Instantaneous(Use Spawn Probability=false, UseLoopCountLimit=false)
  ]
  particleSpawnScript [&InitializeParticle(
    Color Mode=::NewEnumerator0
    Lifetime=5
    Sprite Size Mode=::NewEnumerator3
    Uniform Sprite Size=30
    UsePositionOffset=false
    Write Material Random=true
   )]
  particleUpdateScript [
   &ParticleState(Kill Particles When Lifetime Has Elapsed=false)
   &ScaleSpriteSize(
    Uniform Curve Index=$System.Timeline.Time
    Uniform Curve Sprite Scale=DataInterface() {propertyValues {Curve {keys [{tangentMode=RCTM_User, time=0, value=1}, {tangentMode=RCTM_User, time=0.5, value=1.5}, <at 1.0007 1>]}}}
   )
   &ScaleColor(
    Curve Index=$System.Timeline.Time
    Linear Color Curve=DataInterface() {propertyValues {
      AlphaCurve {keys [<at 0 1>]}
      BlueCurve {keys [<at 0 1>,<at 0.4617 0.025>,<at 1 1>]}
      GreenCurve {keys [<at 0 1>,<at 0.4617 0>,<at 1 1>]}
      RedCurve {keys [<at 0 1>,<at 0.4617 1>,<at 1 1>]}
     }}
    Scale Mode=2
   )
  ]
  rendererClasses [&NiagaraSpriteRendererProperties]
  renderers [{rendererClass=&NiagaraSpriteRendererProperties}]
 }
 Rate {
  emitterUpdateScript [
   &EmitterState()
   &SpawnRate(
    SpawnRate=&FloatFromCurve(
     FloatCurve=DataInterface() {propertyValues {Curve {keys [
         {arriveTangent=6.637, interpMode=RCIM_Cubic, leaveTangent=6.637, tangentMode=RCTM_User, time=0.04, value=0}
         {arriveTangent=16.8545, interpMode=RCIM_Cubic, leaveTangent=-3.293, tangentMode=RCTM_Break, time=0.1, value=1}
         <at 0.9 0>
        ]}}}
     CurveIndex=$System.Timeline.Time
     Scale Curve=50
    )
    Use Spawn Probability=false
   )
  ]
  particleSpawnScript [
   &InitializeParticle(Color Mode=::NewEnumerator0, Sprite Size Mode=::NewEnumerator3, Uniform Sprite Size=8, UsePositionOffset=false, Write Material Random=true)
   &AddVelocity(
    Module.Use Falloff Distance=false
    Override Seed=false
    Velocity Mode=::NewEnumerator1
    Velocity Speed=&RandomRangeFloat(Minimum=120, Maximum=200)
   )
  ]
  particleUpdateScript [
   &ParticleState()
   &Drag(Drag=2, Use Linear Drag=true, Use Rotational Drag=true)
   &GravityForce(Gravity=<0 0 -12>)
   &SolveForcesAndVelocity(Clamp Velocity=false, Limit Acceleration=false)
  ]
  rendererClasses [&NiagaraSpriteRendererProperties]
  renderers [{rendererClass=&NiagaraSpriteRendererProperties}]
 }
 TimeStamp {
  emitterUpdateScript [
   &EmitterState()
   &SpawnBurst_Instantaneous(Use Spawn Probability=false, UseLoopCountLimit=true)
  ]
  particleSpawnScript [&InitializeParticle(Lifetime=5, Sprite Size Mode=::NewEnumerator3, Uniform Sprite Size=50, UsePositionOffset=false, Write Material Random=true)]
  particleUpdateScript [
   &ParticleState(Kill Particles When Lifetime Has Elapsed=false)
   &ScaleSpriteSize(
    Uniform Curve Index=&ClampFloat(Float=&Divide_Float(A=&Subtract_Float(A=$Engine.Time, B=$User.TimeStamp), B=0.4))
    Uniform Curve Sprite Scale=DataInterface() {propertyValues {Curve {keys [
        {interpMode=RCIM_Cubic, tangentMode=RCTM_Break, time=0, value=1}
        {interpMode=RCIM_Cubic, leaveTangent=-9.0426, tangentMode=RCTM_Break, time=0.5403, value=1.5333}
        {interpMode=RCIM_Cubic, leaveTangent=0.2212, tangentMode=RCTM_Break, time=0.7677, value=0.7575}
        {arriveTangent=-4.2629, interpMode=RCIM_Cubic, tangentMode=RCTM_Break, time=1, value=1}
       ]}}}
   )
   &ScaleColor(
    Curve Index=HlslExpression() {hlslExpression="clamp((Engine.Time - User.TimeStamp) / 0.5f, 0, 1)"}
    Linear Color Curve=DataInterface() {propertyValues {
      AlphaCurve {keys [<at 0.5025 1>]}
      BlueCurve {keys [<at 0 1>,<at 0.1654 0.6058>,<at 0.3332 0.0023>,<at 0.6681 0.0374>,<at 0.8637 0.5226>,<at 1 1>]}
      GreenCurve {keys [<at 0 0.2874>,<at 0.1654 0>,<at 0.3332 0>,<at 0.6681 0>,<at 0.8637 0>,<at 1 0.2874>]}
      RedCurve {keys [<at 0 0>,<at 0.1654 0.6788>,<at 0.3332 1>,<at 0.6681 1>,<at 0.8637 0.5885>,<at 1 0>]}
     }}
    Scale Mode=2
   )
  ]
  rendererClasses [&NiagaraSpriteRendererProperties]
  renderers [{rendererClass=&NiagaraSpriteRendererProperties}]
 }
}
summary {
 systemName=NS_Backpack_Template_Elimination
 emitters [
  {emitterName=Burst, rendererClasses [&NiagaraSpriteRendererProperties]}
  {emitterName=Rate, rendererClasses [&NiagaraSpriteRendererProperties]}
  {emitterName=IdleChange, rendererClasses [&NiagaraSpriteRendererProperties]}
  {emitterName=TimeStamp, rendererClasses [&NiagaraSpriteRendererProperties]}
 ]
 userVariables [
  User.bIsFrontEnd {type=NiagaraBool}
  User.bIsFrontEndPreview {type=NiagaraBool}
  User.IsReactive {type=NiagaraFloat, defaultValue=1, description=""}
  User.KillCount {type=NiagaraInt32}
  User.TimeStamp {type=NiagaraFloat, defaultValue=0, description=""}
 ]
}
